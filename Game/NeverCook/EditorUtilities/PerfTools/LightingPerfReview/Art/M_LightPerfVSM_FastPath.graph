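# Unlit UI-domain material for the LightingPerfReview editor perf tool, named for a VSM fast path, that produces an animated traffic-light-style color.
# Time multiplied by a Speed parameter is wrapped with Frac into a repeating 0-1 ramp.
# That ramp drives Lerp_Multiple_Float3, which blends through constant green, yellow and orange-red colors in a looping sweep.
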Material /Game/NeverCook/EditorUtilities/PerfTools/LightingPerfReview/Art/M_LightPerfVSM_FastPath
graph {
 properties {
  bAutomaticallySetUsageInEditor=false
  materialDomain=MD_UI
  pixelDepthOffsetMode=PDOM_Legacy
  shadingModel=MSM_Unlit
  usedShadingModels=Unlit
 }
 nodes [
  n1 = Time()
  n2 = ScalarParameter() {defaultValue=1, parameterName=Speed}
  MaterialFunctionCall(
   1=Constant3Vector() {constant=<rgba 0.1827 0.5208 0.043 1>}
   2=Constant3Vector() {constant=<rgba 0.5 0.4837 0 1>}
   3=Constant3Vector() {constant=<rgba 0.5052 0.0758 0 1>}
   A=Frac(None=Multiply(A=@n1, B=@n2))
  ) {materialFunction=&Lerp_Multiple_Float3}
 ]
}
