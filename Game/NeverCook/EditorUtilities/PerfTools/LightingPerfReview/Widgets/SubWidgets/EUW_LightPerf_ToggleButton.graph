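# Editor utility toggle button for the lighting perf review tool that switches one console variable between enabled and disabled bool, float or string values.
# Clicks write the value from the button's selected state; bool writes just invert that state, ignoring EnabledValue_Bool.
# ValidateActive re-reads the cvar on PreConstruct and after clicks to set label and selection, using a configurable comparison for floats.
# ForceDisable applies the disabled value via console command.
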EditorUtilityWidgetBlueprint /Game/NeverCook/EditorUtilities/PerfTools/LightingPerfReview/Widgets/SubWidgets/EUW_LightPerf_ToggleButton
graphs [
 EventGraph {nodes [
   "On Clicked (TA_Generic01_Button)"!() -> (OutputDelegate, then, widget:User Widget Object Reference, event:Pointer Event Structure) {is=title K2Node_ComponentBoundEvent}
   > HandleButtonClicked()
   Event PreConstruct()
   > ValidateActive()
  ]}
 ForceDisable {nodes [
   ForceDisable() {is=title K2Node_FunctionEntry}
   n1 = "To String (Boolean)"(InBool=NOT Boolean(A=Get EnabledValue_Bool() -> (EnabledValue_Bool:bool)))
   n2 = "To String (Float)"(InDouble=Get DisabledValue_Float() -> (DisabledValue_Float:double))
   n3 = Get ValueType() -> (ValueType:LightingPerfReviewCvarTypes Enum=NewEnumerator0)
   n4 = Get DisabledValue_String() -> (DisabledValue_String:String)
   > ExecuteConsoleCommand(Command=Append!(
     A=Get ConsoleVar() -> (ConsoleVar:String)
     B=" "
     C:String=Select!(NewEnumerator0:String=@n1, NewEnumerator1:String=@n2, NewEnumerator3:String=@n4, Index:LightingPerfReviewCvarTypes Enum=@n3) -> (ReturnValue:String)
    ) -> (ReturnValue))
  ]}
 HandleButtonClicked {nodes [
   HandleButtonClicked() {is=title K2Node_FunctionEntry}
   > n1 = IsValidSetup!(execute, self) -> (then, Valid:bool)
   > Branch(Condition=@n1)
   > Set l_active(l_active:bool=getSelected!(self:TA Base Editor Button Widget Object Reference=Get TA_Generic01_Button() -> (TA_Generic01_Button:TA Generic 01 Button Object Reference)) -> (isSelected:bool)) -> (Output_Get:bool)
   > n2 = Sequence()
   > n3 = Switch on LightingPerfReviewCvarTypes!(execute, Selection:LightingPerfReviewCvarTypes Enum=Get ValueType() -> (ValueType=NewEnumerator0)) -> (NewEnumerator0:Exec, NewEnumerator1:Exec, NewEnumerator3:Exec) {is=title K2Node_SwitchEnum}
   > "Set Console Variable By Name (Bool)"!(execute, InCommandName:String=Get ConsoleVar(), InValue:bool=NOT Boolean(A=Get l_active() -> (l_active:bool))) -> (then, ReturnValue:bool) {is=title K2Node_CallFunction}
   n4 = Get ConsoleVar()
   "Set Console Variable By Name (Float)"!(
    execute=@n3.NewEnumerator1
    InCommandName:String=@n4
    InValue:float=SelectFloat(A=Get DisabledValue_Float(), B=Get EnabledValue_Float() -> (EnabledValue_Float:double), bPickA=Get l_active())
   ) -> (then, ReturnValue:bool) {is=title K2Node_CallFunction}
   "Set Console Variable By Name (String)"!(
    execute=@n3.NewEnumerator3
    InCommandName:String=@n4
    InValue:String=Select!(
     Option 0:String=Get EnabledValue_String() -> (EnabledValue_String:String)
     Option 1:String=Get DisabledValue_String()
     Index:bool=Get l_active()
    ) -> (ReturnValue)
   ) -> (then, ReturnValue:bool) {is=title K2Node_CallFunction}
   ValidateActive(execute=@n2.then_1)
  ]}
 IsActive {nodes [
   IsActive() {is=title K2Node_FunctionEntry}
   > n1 = IsValidSetup()
   > n2 = Branch(Condition=@n1)
   > n3 = Sequence()
   > n4 = Switch on LightingPerfReviewCvarTypes(Selection=Get ValueType() -> (ValueType=NewEnumerator0))
   n5 = Get ConsoleVar()
   > n6 = GetConsoleVariableBoolValue!(execute, VariableName:String=@n5) -> (then, ReturnValue:bool)
   > Set l_active(l_active="Equal (Boolean)"(A:bool=@n6, B:bool=Get EnabledValue_Bool()) -> (ReturnValue:bool=~) {is=title K2Node_PromotableOperator})
   n7 = GetConsoleVariableFloatValue!(execute=@n4.NewEnumerator1, VariableName:String=@n5) -> (then, ReturnValue:float)
   > Set l_float(l_float:float=@n7) -> (Output_Get:float)
   n8 = Get l_float() -> (l_float:float)
   n9 = Get EnabledValue_Float()
   > Set l_active(l_active=Select!(
     Dataflow_Compare_Equal:bool="Equal (Float)"(A=@n8, B=@n9)
     Dataflow_Compare_Smaller:bool="float < float"(A=@n8, B=@n9)
     Dataflow_Compare_SmallerOrEqual:bool="float <= float"(A=@n8, B=@n9) -> (ReturnValue=~)
     Dataflow_Compare_Greater:bool="float > float"(A=@n8, B=@n9)
     Dataflow_Compare_GreaterOrEqual:bool="float >= float"(A=@n8, B=@n9)
     Dataflow_Compare_NotEqual:bool=false
     Index:ECompareOperationEnum Enum=Get EnabledValue_Float_Comparison() -> (EnabledValue_Float_Comparison:ECompareOperationEnum Enum=Dataflow_Compare_Equal)
    ) -> (ReturnValue:bool))
   n10 = GetConsoleVariableStringValue!(execute=@n4.NewEnumerator3, InCommandName:String=@n5) -> (then, OutValue:String, ReturnValue:bool)
   > Branch(Condition=@n10.ReturnValue)
   > Set l_active(l_active="Equal Exactly (String)"(A=@n10, B=Get EnabledValue_String()))
   Return Node!(execute=@n3.then_1, Active:bool=Get l_active())
   Return Node(execute=@n2.else, Active=false)
  ]}
 IsValidSetup {nodes [
   IsValidSetup() {is=title K2Node_FunctionEntry}
   n1 = Get ConsoleVar()
   n2 = IsEmpty(InString=@n1)
   n3 = "Equal Exactly (String)"(A=@n1, B=None)
   > Return Node!(execute, Valid:bool=NOT Boolean(A=OR Boolean(A=@n2, B=@n3)))
  ]}
 ValidateActive {nodes [
   ValidateActive() {is=title K2Node_FunctionEntry}
   > n1 = IsActive!(execute, self) -> (then, Active:bool) {is=title K2Node_CallFunction}
   n2 = Get TA_Generic01_Button()
   > setLabel!(
    execute
    self:TA Base Editor Button Widget Object Reference=@n2
    label:String=Append!(A=Get Label() -> (Label:String), B=Select!(Option 0:String=" OFF", Option 1:String=" ACTIVE", Index:bool=@n1) -> (ReturnValue:String)) -> (ReturnValue)
   ) -> (then)
   > setSelected!(execute, self:TA Base Editor Button Widget Object Reference=@n2, selected:bool=@n1, throwEvent:bool=false) -> (then)
  ]}
]
variables [
 ConsoleVar {type=string, default=""}
 DisabledValue_Float {type=number}
 DisabledValue_String {type=string, default=""}
 EnabledValue_Bool {type=boolean, default=true}
 EnabledValue_Float {type=number, default=1}
 EnabledValue_Float_Comparison {type=ECompareOperationEnum, default=Dataflow_Compare_Equal}
 EnabledValue_String {type=string, default=""}
 Label {type=string, default=Toggle}
 ValueType {type=LightingPerfReviewCvarTypes, default=Bool}
]
widgets {
 info {rootWidgetClass=&StackBox, widgetCount=2}
 widgets [
  StackBox_0 {widgetClassPath=&StackBox, widget=StackBox_0}
  TA_Generic01_Button {
   widgetClassPath=&TA_Generic01_Button
   parent=StackBox_0
   widget=TA_Generic01_Button
   properties {label=Toggle, buttonColor_Default=<rgba 0.4635 0.0303 0 1>, buttonColor_Selected=<rgba 0.2318 0.4635 0 1>}
   slot {class=StackBoxSlot, properties {size {sizeRule=Fill, value=1}}}
  }
 ]
}
