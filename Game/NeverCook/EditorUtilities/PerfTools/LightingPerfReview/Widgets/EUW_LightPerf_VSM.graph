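# Editor utility panel in the lighting perf review tool for inspecting Virtual Shadow Map and Nanite costs.
# Toggle-button subwidgets drive a VSM debug show flag and Nanite visualize modes (static, mask, WPO), with color legends beneath.
# Construct starts a looping timer and enables advanced Nanite visualization; Destruct force-disables all toggles and turns it off.
# DisableAll and ValidateAll find toggles by class-filtering the vertical box's children; SetGlobalsEnabled sets the advanced cvar.
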EditorUtilityWidgetBlueprint /Game/NeverCook/EditorUtilities/PerfTools/LightingPerfReview/Widgets/EUW_LightPerf_VSM
graphs [
 DisableAll {nodes [
   DisableAll() {is=title K2Node_FunctionEntry}
   > n1 = Filter by Class!(
    execute
    TargetArray:Array of Object References=GetAllChildren(self=Get VerticalBox_0() -> (VerticalBox_0:Vertical Box Object Reference))
    ObjectClass:Object Class Reference=/Game/NeverCook/EditorUtilities/PerfTools/LightingPerfReview/Widgets/SubWidgets/EUW_LightPerf_ToggleButton.EUW_LightPerf_ToggleButton_C
    FilterType:EEditorScriptingFilterType Enum
   ) -> (then, ReturnValue:Array of EUW Light Perf Toggle Button Object References)
   > n2 = For Each Loop(Array:Array of EUW Light Perf Toggle Button Object References=@n1) -> (Array Element:EUW Light Perf Toggle Button Object Reference)
   > ForceDisable(self:EUW Light Perf Toggle Button Object Reference=@n2)
  ]}
 EventGraph {nodes [
   Event Construct()
   > Set Timer by Event(Delegate=Create Event(), Time=0.25, bLooping=true, bMaxOncePerFrame=true, InitialStartDelay=1)
   > ExecuteConsoleCommand(Command=r.nanite.visualize.advanced 1)
   Event Destruct()
   > DisableAll() {is=title K2Node_CallFunction}
   > ExecuteConsoleCommand(Command=r.nanite.visualize.advanced 0)
   Event PreConstruct()
  ]}
 SetGlobalsEnabled {nodes [
   n1 = SetGlobalsEnabled!() -> (then, Enabled:bool=~) {is=title K2Node_FunctionEntry}
   > Set l_enabled(l_enabled:bool=@n1) -> (Output_Get:bool)
   > "Set Console Variable By Name (Bool)"!(execute, InCommandName:String=r.nanite.visualize.advanced, InValue:bool=Get l_enabled() -> (l_enabled:bool)) -> (then, ReturnValue:bool) {is=title K2Node_CallFunction}
  ]}
 ValidateAll {nodes [
   ValidateAll() {is=title K2Node_FunctionEntry}
   > n1 = Filter by Class(
    TargetArray=GetAllChildren(self=Get VerticalBox_0())
    ObjectClass=&EUW_LightPerf_ToggleButton_C
   )
   > n2 = For Each Loop(Array=@n1)
   > ValidateActive(self:EUW Light Perf Toggle Button Object Reference=@n2)
  ]}
]
variables []
widgets {
 info {rootWidgetClass=&StackBox, widgetCount=26}
 widgets [
  StackBox_0 {widgetClassPath=&StackBox, widget=StackBox_0}
  VerticalBox_0 {
   widgetClassPath=&VerticalBox
   bIsVariable=true
   parent=StackBox_0
   widget=VerticalBox_0
   slot {class=StackBoxSlot, properties {size {sizeRule=Fill, value=1}}}
  }
  EUW_LightPerf_ToggleButton {
   widgetClassPath=/Game/NeverCook/EditorUtilities/PerfTools/LightingPerfReview/Widgets/SubWidgets/EUW_LightPerf_ToggleButton
   bIsVariable=true
   parent=VerticalBox_0
   widget=EUW_LightPerf_ToggleButton
   properties {consoleVar=ShowFlag.DebugDrawDistantVirtualSMLights, label=Debug Fast Path, valueType=Float}
   slot {class=VerticalBoxSlot}
  }
  HorizontalBox_179 {widgetClassPath=&HorizontalBox, parent=VerticalBox_0, widget=HorizontalBox_179, slot {class=VerticalBoxSlot}}
  Image_77 {
   widgetClassPath=&Image
   bIsVariable=true
   parent=HorizontalBox_179
   widget=Image_77
   properties {brush {
     bIsDynamicallyLoaded=false
     drawAs=Image
     imageType=NoImage
     mirroring=NoMirror
     resourceName=None
     resourceObject=null
     tiling=NoTile
     imageSize=<16 16>
     margin=<ltrb 0 0 0 0>
     outlineSettings {
      bUseBrushTransparency=false
      roundingType=HalfHeightRadius
      width=0
      color {colorUseRule=UseColor_Specified, specifiedColor=<rgba 0 0 0 0>}
      cornerRadii=<0 0 0 0>
     }
     tintColor {colorUseRule=UseColor_Specified, specifiedColor=<rgba 0 0.1838 1 1>}
     uVRegion {bIsValid=false, max=<0 0>, min=<0 0>}
    }}
   slot {class=HorizontalBoxSlot, properties {padding=<ltrb 5 5 5 5>}}
  }
  CommonTextBlock_130 {
   widgetClassPath=&CommonTextBlock
   parent=HorizontalBox_179
   widget=CommonTextBlock_130
   properties {
    text="( Solid Blue ) Slow"
    font {
     bForceMonospaced=false
     bMaterialIsStencil=false
     fontMaterial=null
     fontObject=&NotoSans
     letterSpacing=0
     monospacedWidth=1
     size=9
     skewAmount=0
     typefaceFontName=NotoSans-Regular
     outlineSettings {
      bApplyOutlineToDropShadows=false
      bMiteredCorners=false
      bSeparateFillAlpha=false
      outlineMaterial=null
      outlineSize=0
      outlineColor=<rgba 0 0 0 1>
     }
    }
    shadowOffset=<0 0>
   }
   slot {class=HorizontalBoxSlot, properties {verticalAlignment=VAlign_Center, padding=<ltrb 5 0 0 0>}}
  }
  Image {
   widgetClassPath=&Image
   bIsVariable=true
   parent=HorizontalBox_179
   widget=Image
   properties {brush {
     bIsDynamicallyLoaded=false
     drawAs=Image
     imageType=FullColor
     mirroring=NoMirror
     resourceName=None
     resourceObject=/Game/NeverCook/EditorUtilities/PerfTools/LightingPerfReview/Art/M_LightPerfVSM_FastPath
     tiling=NoTile
     imageSize=<16 16>
     margin=<ltrb 0 0 0 0>
     outlineSettings {
      bUseBrushTransparency=false
      roundingType=HalfHeightRadius
      width=0
      color {colorUseRule=UseColor_Specified, specifiedColor=<rgba 0 0 0 0>}
      cornerRadii=<0 0 0 0>
     }
     tintColor {colorUseRule=UseColor_Specified, specifiedColor=<rgba 1 1 1 1>}
     uVRegion {bIsValid=false, max=<0 0>, min=<0 0>}
    }}
   slot {class=HorizontalBoxSlot, properties {padding=<ltrb 10 5 5 5>}}
  }
  CommonTextBlock {
   widgetClassPath=&CommonTextBlock
   parent=HorizontalBox_179
   widget=CommonTextBlock
   properties {
    text="( Flashing ) Fast Path"
    font {
     bForceMonospaced=false
     bMaterialIsStencil=false
     fontMaterial=null
     fontObject=&NotoSans
     letterSpacing=0
     monospacedWidth=1
     size=9
     skewAmount=0
     typefaceFontName=NotoSans-Regular
     outlineSettings {
      bApplyOutlineToDropShadows=false
      bMiteredCorners=false
      bSeparateFillAlpha=false
      outlineMaterial=null
      outlineSize=0
      outlineColor=<rgba 0 0 0 1>
     }
    }
    shadowOffset=<0 0>
   }
   slot {class=HorizontalBoxSlot, properties {verticalAlignment=VAlign_Center, padding=<ltrb 5 0 0 0>}}
  }
  EUW_LightPerf_ToggleButton_1 {
   widgetClassPath=&EUW_LightPerf_ToggleButton
   bIsVariable=true
   parent=VerticalBox_0
   widget=EUW_LightPerf_ToggleButton_1
   properties {consoleVar=r.nanite.visualize, disabledValue_String=none, enabledValue_String=vsmstatic, label=VSM Static, valueType=String}
   slot {class=VerticalBoxSlot}
  }
  HorizontalBox {widgetClassPath=&HorizontalBox, parent=VerticalBox_0, widget=HorizontalBox, slot {class=VerticalBoxSlot}}
  Image_1 {
   widgetClassPath=&Image
   bIsVariable=true
   parent=HorizontalBox
   widget=Image_1
   properties {brush {
     bIsDynamicallyLoaded=false
     drawAs=Image
     imageType=NoImage
     mirroring=NoMirror
     resourceName=None
     resourceObject=null
     tiling=NoTile
     imageSize=<16 16>
     margin=<ltrb 0 0 0 0>
     outlineSettings {
      bUseBrushTransparency=false
      roundingType=HalfHeightRadius
      width=0
      color {colorUseRule=UseColor_Specified, specifiedColor=<rgba 0 0 0 0>}
      cornerRadii=<0 0 0 0>
     }
     tintColor {colorUseRule=UseColor_Specified, specifiedColor=<rgba 0 0.1838 1 1>}
     uVRegion {bIsValid=false, max=<0 0>, min=<0 0>}
    }}
   slot {class=HorizontalBoxSlot, properties {padding=<ltrb 5 5 5 5>}}
  }
  CommonTextBlock_1 {
   widgetClassPath=&CommonTextBlock
   parent=HorizontalBox
   widget=CommonTextBlock_1
   properties {
    text="Dynamic (Bad)"
    font {
     bForceMonospaced=false
     bMaterialIsStencil=false
     fontMaterial=null
     fontObject=&NotoSans
     letterSpacing=0
     monospacedWidth=1
     size=9
     skewAmount=0
     typefaceFontName=NotoSans-Regular
     outlineSettings {
      bApplyOutlineToDropShadows=false
      bMiteredCorners=false
      bSeparateFillAlpha=false
      outlineMaterial=null
      outlineSize=0
      outlineColor=<rgba 0 0 0 1>
     }
    }
    shadowOffset=<0 0>
   }
   slot {class=HorizontalBoxSlot, properties {verticalAlignment=VAlign_Center, padding=<ltrb 5 0 0 0>}}
  }
  Image_2 {
   widgetClassPath=&Image
   bIsVariable=true
   parent=HorizontalBox
   widget=Image_2
   properties {brush {
     bIsDynamicallyLoaded=false
     drawAs=Image
     imageType=FullColor
     mirroring=NoMirror
     resourceName=None
     resourceObject=null
     tiling=NoTile
     imageSize=<16 16>
     margin=<ltrb 0 0 0 0>
     outlineSettings {
      bUseBrushTransparency=false
      roundingType=HalfHeightRadius
      width=0
      color {colorUseRule=UseColor_Specified, specifiedColor=<rgba 0 0 0 0>}
      cornerRadii=<0 0 0 0>
     }
     tintColor {colorUseRule=UseColor_Specified, specifiedColor=<rgba 0.5312 0.0083 0.0083 1>}
     uVRegion {bIsValid=false, max=<0 0>, min=<0 0>}
    }}
   slot {class=HorizontalBoxSlot, properties {padding=<ltrb 10 5 5 5>}}
  }
  CommonTextBlock_2 {
   widgetClassPath=&CommonTextBlock
   parent=HorizontalBox
   widget=CommonTextBlock_2
   properties {
    text="Static (Good)"
    font {
     bForceMonospaced=false
     bMaterialIsStencil=false
     fontMaterial=null
     fontObject=&NotoSans
     letterSpacing=0
     monospacedWidth=1
     size=9
     skewAmount=0
     typefaceFontName=NotoSans-Regular
     outlineSettings {
      bApplyOutlineToDropShadows=false
      bMiteredCorners=false
      bSeparateFillAlpha=false
      outlineMaterial=null
      outlineSize=0
      outlineColor=<rgba 0 0 0 1>
     }
    }
    shadowOffset=<0 0>
   }
   slot {class=HorizontalBoxSlot, properties {verticalAlignment=VAlign_Center, padding=<ltrb 5 0 0 0>}}
  }
  EUW_LightPerf_ToggleButton_2 {
   widgetClassPath=&EUW_LightPerf_ToggleButton
   bIsVariable=true
   parent=VerticalBox_0
   widget=EUW_LightPerf_ToggleButton_2
   properties {consoleVar=r.nanite.visualize, disabledValue_String=None, enabledValue_String=Mask, label=Nanite Mask, valueType=String}
   slot {class=VerticalBoxSlot}
  }
  HorizontalBox_1 {widgetClassPath=&HorizontalBox, parent=VerticalBox_0, widget=HorizontalBox_1, slot {class=VerticalBoxSlot}}
  Image_3 {
   widgetClassPath=&Image
   bIsVariable=true
   parent=HorizontalBox_1
   widget=Image_3
   properties {brush {
     bIsDynamicallyLoaded=false
     drawAs=Image
     imageType=NoImage
     mirroring=NoMirror
     resourceName=None
     resourceObject=null
     tiling=NoTile
     imageSize=<16 16>
     margin=<ltrb 0 0 0 0>
     outlineSettings {
      bUseBrushTransparency=false
      roundingType=HalfHeightRadius
      width=0
      color {colorUseRule=UseColor_Specified, specifiedColor=<rgba 0 0 0 0>}
      cornerRadii=<0 0 0 0>
     }
     tintColor {colorUseRule=UseColor_Specified, specifiedColor=<rgba 0.7865 0 0.0185 1>}
     uVRegion {bIsValid=false, max=<0 0>, min=<0 0>}
    }}
   slot {class=HorizontalBoxSlot, properties {padding=<ltrb 5 5 5 5>}}
  }
  CommonTextBlock_3 {
   widgetClassPath=&CommonTextBlock
   parent=HorizontalBox_1
   widget=CommonTextBlock_3
   properties {
    text="Nanite Disabled (Often Bad)"
    font {
     bForceMonospaced=false
     bMaterialIsStencil=false
     fontMaterial=null
     fontObject=&NotoSans
     letterSpacing=0
     monospacedWidth=1
     size=9
     skewAmount=0
     typefaceFontName=NotoSans-Regular
     outlineSettings {
      bApplyOutlineToDropShadows=false
      bMiteredCorners=false
      bSeparateFillAlpha=false
      outlineMaterial=null
      outlineSize=0
      outlineColor=<rgba 0 0 0 1>
     }
    }
    shadowOffset=<0 0>
   }
   slot {class=HorizontalBoxSlot, properties {verticalAlignment=VAlign_Center, padding=<ltrb 5 0 0 0>}}
  }
  Image_4 {
   widgetClassPath=&Image
   bIsVariable=true
   parent=HorizontalBox_1
   widget=Image_4
   properties {brush {
     bIsDynamicallyLoaded=false
     drawAs=Image
     imageType=FullColor
     mirroring=NoMirror
     resourceName=None
     resourceObject=null
     tiling=NoTile
     imageSize=<16 16>
     margin=<ltrb 0 0 0 0>
     outlineSettings {
      bUseBrushTransparency=false
      roundingType=HalfHeightRadius
      width=0
      color {colorUseRule=UseColor_Specified, specifiedColor=<rgba 0 0 0 0>}
      cornerRadii=<0 0 0 0>
     }
     tintColor {colorUseRule=UseColor_Specified, specifiedColor=<rgba 0.0427 0.5312 0 1>}
     uVRegion {bIsValid=false, max=<0 0>, min=<0 0>}
    }}
   slot {class=HorizontalBoxSlot, properties {padding=<ltrb 10 5 5 5>}}
  }
  CommonTextBlock_4 {
   widgetClassPath=&CommonTextBlock
   parent=HorizontalBox_1
   widget=CommonTextBlock_4
   properties {
    text="Nanite  Enabled (Usually Good)"
    font {
     bForceMonospaced=false
     bMaterialIsStencil=false
     fontMaterial=null
     fontObject=&NotoSans
     letterSpacing=0
     monospacedWidth=1
     size=9
     skewAmount=0
     typefaceFontName=NotoSans-Regular
     outlineSettings {
      bApplyOutlineToDropShadows=false
      bMiteredCorners=false
      bSeparateFillAlpha=false
      outlineMaterial=null
      outlineSize=0
      outlineColor=<rgba 0 0 0 1>
     }
    }
    shadowOffset=<0 0>
   }
   slot {class=HorizontalBoxSlot, properties {verticalAlignment=VAlign_Center, padding=<ltrb 5 0 0 0>}}
  }
  EUW_LightPerf_ToggleButton_3 {
   widgetClassPath=&EUW_LightPerf_ToggleButton
   bIsVariable=true
   parent=VerticalBox_0
   widget=EUW_LightPerf_ToggleButton_3
   properties {consoleVar=r.nanite.visualize, disabledValue_String=None, enabledValue_String=EvaluateWPO, label=Nanite Eval WPO, valueType=String}
   slot {class=VerticalBoxSlot}
  }
  HorizontalBox_2 {widgetClassPath=&HorizontalBox, parent=VerticalBox_0, widget=HorizontalBox_2, slot {class=VerticalBoxSlot}}
  Image_5 {
   widgetClassPath=&Image
   bIsVariable=true
   parent=HorizontalBox_2
   widget=Image_5
   properties {brush {
     bIsDynamicallyLoaded=false
     drawAs=Image
     imageType=NoImage
     mirroring=NoMirror
     resourceName=None
     resourceObject=null
     tiling=NoTile
     imageSize=<16 16>
     margin=<ltrb 0 0 0 0>
     outlineSettings {
      bUseBrushTransparency=false
      roundingType=HalfHeightRadius
      width=0
      color {colorUseRule=UseColor_Specified, specifiedColor=<rgba 0 0 0 0>}
      cornerRadii=<0 0 0 0>
     }
     tintColor {colorUseRule=UseColor_Specified, specifiedColor=<rgba 0.7865 0 0.0185 1>}
     uVRegion {bIsValid=false, max=<0 0>, min=<0 0>}
    }}
   slot {class=HorizontalBoxSlot, properties {padding=<ltrb 5 5 5 5>}}
  }
  CommonTextBlock_5 {
   widgetClassPath=&CommonTextBlock
   parent=HorizontalBox_2
   widget=CommonTextBlock_5
   properties {
    text="WPO Disabled (Good)"
    font {
     bForceMonospaced=false
     bMaterialIsStencil=false
     fontMaterial=null
     fontObject=&NotoSans
     letterSpacing=0
     monospacedWidth=1
     size=9
     skewAmount=0
     typefaceFontName=NotoSans-Regular
     outlineSettings {
      bApplyOutlineToDropShadows=false
      bMiteredCorners=false
      bSeparateFillAlpha=false
      outlineMaterial=null
      outlineSize=0
      outlineColor=<rgba 0 0 0 1>
     }
    }
    shadowOffset=<0 0>
   }
   slot {class=HorizontalBoxSlot, properties {verticalAlignment=VAlign_Center, padding=<ltrb 5 0 0 0>}}
  }
  Image_6 {
   widgetClassPath=&Image
   bIsVariable=true
   parent=HorizontalBox_2
   widget=Image_6
   properties {brush {
     bIsDynamicallyLoaded=false
     drawAs=Image
     imageType=FullColor
     mirroring=NoMirror
     resourceName=None
     resourceObject=null
     tiling=NoTile
     imageSize=<16 16>
     margin=<ltrb 0 0 0 0>
     outlineSettings {
      bUseBrushTransparency=false
      roundingType=HalfHeightRadius
      width=0
      color {colorUseRule=UseColor_Specified, specifiedColor=<rgba 0 0 0 0>}
      cornerRadii=<0 0 0 0>
     }
     tintColor {colorUseRule=UseColor_Specified, specifiedColor=<rgba 0.0427 0.5312 0 1>}
     uVRegion {bIsValid=false, max=<0 0>, min=<0 0>}
    }}
   slot {class=HorizontalBoxSlot, properties {padding=<ltrb 10 5 5 5>}}
  }
  CommonTextBlock_6 {
   widgetClassPath=&CommonTextBlock
   parent=HorizontalBox_2
   widget=CommonTextBlock_6
   properties {
    text="WPO Enabled (Bad)"
    font {
     bForceMonospaced=false
     bMaterialIsStencil=false
     fontMaterial=null
     fontObject=&NotoSans
     letterSpacing=0
     monospacedWidth=1
     size=9
     skewAmount=0
     typefaceFontName=NotoSans-Regular
     outlineSettings {
      bApplyOutlineToDropShadows=false
      bMiteredCorners=false
      bSeparateFillAlpha=false
      outlineMaterial=null
      outlineSize=0
      outlineColor=<rgba 0 0 0 1>
     }
    }
    shadowOffset=<0 0>
   }
   slot {class=HorizontalBoxSlot, properties {verticalAlignment=VAlign_Center, padding=<ltrb 5 0 0 0>}}
  }
 ]
}
